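# Physics asset for the right foot of the female medium Durrburger PJs cosmetic body, covering the dynamic eye and tongue bones on the foot.
# A small capsule on ball_r anchors two eye spheres and a two-box tongue chain, with the second tongue segment constrained to the first.
# Constraints give the eyes wider swing with zero twist limit, while the tongue segments get tighter swing and slight twist.
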PhysicsAsset /BRCosmetics/Characters/Player/Female/Medium/Bodies/F_MED_DurrburgerPJs/Meshes/F_MED_DurrburgerPJs_01_FootR_Physics
bodies {
 ball_r [{shapeType=Capsule, length=0.505, radius=0.505, rotation=<rot 0 -90 0>}]
 dyn_foot_eye_inner_r [{shapeType=Sphere, radius=2.2, center=<0.01 0 0>}]
 dyn_foot_eye_outer_r [{shapeType=Sphere, radius=2.2, center=<0.01 0 0>}]
 dyn_foot_tongue_01_r [{shapeType=Box, extentX=2.6396, extentY=1.6042, extentZ=5.9117, center=<1.4267 -0.1598 0.0202>, rotation=<rot 0 -177.7425 -88.6939>}]
 dyn_foot_tongue_02_r [{shapeType=Box, extentX=1.9344, extentY=1.553, extentZ=5.2096, center=<1.5158 -0.2622 0.0069>, rotation=<rot 0 -171.9745 -89.0108>}]
}
constraints [
 {bone1Name=dyn_foot_tongue_01_r, bone2Name=ball_r, swing1LimitDegrees=45, swing2LimitDegrees=45, twistLimitDegrees=10}
 {bone1Name=dyn_foot_tongue_02_r, bone2Name=dyn_foot_tongue_01_r, swing1LimitDegrees=45, swing2LimitDegrees=45, twistLimitDegrees=10}
 {bone1Name=dyn_foot_eye_outer_r, bone2Name=ball_r, swing1LimitDegrees=60, swing2LimitDegrees=60, twistLimitDegrees=0}
 {bone1Name=dyn_foot_eye_inner_r, bone2Name=ball_r, swing1LimitDegrees=60, swing2LimitDegrees=60, twistLimitDegrees=0}
]
