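# Post-process style Anim Blueprint for the female medium Durrburger PJs cosmetic body, adding secondary motion to costume bones.
# It copies the attached parent's pose through a Control Rig, then runs per-foot RigidBody simulations and a half-weight sprung AnimDynamics on dyn_cup.
# Weighted constraints then drive dyn_shell_01–03 from left clavicle, upper arm and spine bones.
# Right-side constraints for dyn_shell_04–06 exist but are not wired to the output; event logic is empty.
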AnimBlueprint /BRCosmetics/Characters/Player/Female/Medium/Bodies/F_MED_DurrburgerPJs/Meshes/F_MED_DurrburgerPJs_01_Body_AnimBP
target_skeleton=&/F_MED_DurrburgerPJs_01_Body_Skeleton
graphs [
 AnimGraph {nodes [
   n1 = "Constraint - dyn_shell_04 (PPP)"!(ComponentPose, Alpha=1, ConstraintWeights_0=0.8167, ConstraintWeights_1:float=0.1208, ConstraintWeights_2:float=0.0625) -> (Pose) {
    alphaBoolBlend {blendOption=Linear}
    boneToModify {boneName=dyn_shell_04}
    constraintSetup [{targetBone {boneName=clavicle_r}}, {targetBone {boneName=spine_04}}, {targetBone {boneName=spine_05}}]
    constraintWeights [1,1,1]
   }
   n2 = "Constraint - dyn_shell_05 (PPP)"!(ComponentPose=@n1, Alpha=1, ConstraintWeights_0=0.5738, ConstraintWeights_1:float=0.2048, ConstraintWeights_2:float=0.2214) -> (Pose) {
    alphaBoolBlend {blendOption=Linear}
    boneToModify {boneName=dyn_shell_05}
    constraintSetup [{targetBone {boneName=clavicle_r}}, {targetBone {boneName=spine_04}}, {targetBone {boneName=spine_05}}]
    constraintWeights [1,1,1]
   }
   "Constraint - dyn_shell_06 (PPPPPP)"!(
    ComponentPose=@n2
    Alpha=1
    ConstraintWeights_0=0
    ConstraintWeights_1:float=0.0801
    ConstraintWeights_2:float=0
    ConstraintWeights_3:float=0.5865
    ConstraintWeights_4:float=0.178
    ConstraintWeights_5:float=0.1553
   ) -> (Pose) {
    alphaBoolBlend {blendOption=Linear}
    boneToModify {boneName=dyn_shell_06}
    constraintSetup [
     {targetBone {boneName=upperarm_r}}
     {targetBone {boneName=upperarm_twist_01_r}}
     {targetBone {boneName=upperarm_twist_02_r}}
     {targetBone {boneName=spine_04}}
     {targetBone {boneName=spine_05}}
     {targetBone {boneName=clavicle_r}}
    ]
    constraintWeights [1,1,1,1,1,1]
   }
   n3 = Control Rig(Source=Copy Pose From Mesh() {bUseAttachedParent=true}) {
    alphaBoolBlend {blendOption=Linear}
    controlRigAssetReference {blueprintRigClass=&/F_MED_DurrburgerPJs_01_Body_CtrlRig.F_MED_DurrburgerPJs_01_Body_CtrlRig_C}
   }
   n4 = RigidBody(ComponentPose=Local To Component(LocalPose=@n3)) {
    bClampLinearTranslationLimitToRefPose=true
    bOverrideWorldGravity=true
    overridePhysicsAsset=&/F_MED_DurrburgerPJs_01_FootR_Physics
    simulationSpace=BaseBoneSpace
    alphaBoolBlend {blendOption=Linear}
    baseBoneRef {boneName=foot_r}
    overrideWorldGravity {z=-980}
    simSpaceSettings {dampingAlpha=0, maxAngularAcceleration=400, maxAngularVelocity=600, maxLinearAcceleration=200, maxLinearVelocity=1400, worldAlpha=1}
   }
   n5 = RigidBody(ComponentPose=@n4) {
    bClampLinearTranslationLimitToRefPose=true
    bOverrideWorldGravity=true
    overridePhysicsAsset=&/F_MED_DurrburgerPJs_01_FootL_Physics
    simulationSpace=BaseBoneSpace
    alphaBoolBlend {blendOption=Linear}
    baseBoneRef {boneName=foot_l}
    overrideWorldGravity {z=-980}
    simSpaceSettings {dampingAlpha=0, maxAngularAcceleration=400, maxAngularVelocity=600, maxLinearAcceleration=200, maxLinearVelocity=1400, worldAlpha=1}
   }
   n6 = Anim Dynamics - Bone: dyn_cup(ComponentPose=@n5, Alpha=0.5, ExternalForce="0,0,0") {
    angularDampingOverride=1
    angularSpringConstant=4000
    bAngularSpring=true
    bOverrideAngularDamping=true
    bUseGravityOverride=true
    gravityScale=0.35
    simulationSpace=RootRelative
    alphaBoolBlend {blendOption=Linear}
    boundBone {boneName=dyn_cup}
    gravityOverride {z=-980}
    physicsBodyDefinitions [{
      boundBone {boneName=dyn_cup}
      boxExtents=<20 10 10>
      constraintSetup {angularLimitsMax=<2 2 22.5>, angularLimitsMin=<-2 -2 -22.5>, angularTarget=<1 0 0>}
      localJointOffset=<2 0 0>
     }]
    simSpaceSettings {maxAngularAcceleration=300, maxAngularVelocity=770, simSpaceAngularAlpha=1}
   }
   n7 = "Constraint - dyn_shell_03 (PPPPPP)"!(
    ComponentPose=@n6
    Alpha=1
    ConstraintWeights_0=0.2162
    ConstraintWeights_1:float=0.0005
    ConstraintWeights_2:float=0.495
    ConstraintWeights_3:float=0.2323
    ConstraintWeights_4:float=0.0555
    ConstraintWeights_5:float=0.0005
   ) -> (Pose) {
    alphaBoolBlend {blendOption=Linear}
    boneToModify {boneName=dyn_shell_03}
    constraintSetup [
     {targetBone {boneName=clavicle_l}}
     {targetBone {boneName=upperarm_twist_02_l}}
     {targetBone {boneName=spine_04}}
     {targetBone {boneName=spine_05}}
     {targetBone {boneName=upperarm_twist_01_l}}
     {targetBone {boneName=upperarm_l}}
    ]
    constraintWeights [1,1,1,1,1,1]
   }
   n8 = "Constraint - dyn_shell_02 (PPPPPP)"!(
    ComponentPose=@n7
    Alpha=1
    ConstraintWeights_0=0.4986
    ConstraintWeights_1:float=0.0002
    ConstraintWeights_2:float=0.2307
    ConstraintWeights_3:float=0.2156
    ConstraintWeights_4:float=0.0548
    ConstraintWeights_5:float=0.0002
   ) -> (Pose) {
    alphaBoolBlend {blendOption=Linear}
    boneToModify {boneName=dyn_shell_02}
    constraintSetup [
     {targetBone {boneName=clavicle_l}}
     {targetBone {boneName=upperarm_twist_02_l}}
     {targetBone {boneName=spine_04}}
     {targetBone {boneName=spine_05}}
     {targetBone {boneName=upperarm_twist_01_l}}
     {targetBone {boneName=upperarm_l}}
    ]
    constraintWeights [1,1,1,1,1,1]
   }
   n9 = "Constraint - dyn_shell_01 (PPPPPP)"!(
    ComponentPose=@n8
    Alpha=1
    ConstraintWeights_0=0.8048
    ConstraintWeights_1:float=0.0002
    ConstraintWeights_2:float=0.0855
    ConstraintWeights_3:float=0.0638
    ConstraintWeights_4:float=0.0456
    ConstraintWeights_5:float=0.0001
   ) -> (Pose) {
    alphaBoolBlend {blendOption=Linear}
    boneToModify {boneName=dyn_shell_01}
    constraintSetup [
     {targetBone {boneName=clavicle_l}}
     {targetBone {boneName=upperarm_twist_02_l}}
     {targetBone {boneName=spine_04}}
     {targetBone {boneName=spine_05}}
     {targetBone {boneName=upperarm_twist_01_l}}
     {targetBone {boneName=upperarm_l}}
    ]
    constraintWeights [1,1,1,1,1,1]
   }
   Output Pose(Result=Component To Local(ComponentPose=@n9)) {layerGroup=None}
  ]}
 EventGraph {nodes [
   TryGetPawnOwner()
   Event BlueprintUpdateAnimation()
  ]}
]
variables []
