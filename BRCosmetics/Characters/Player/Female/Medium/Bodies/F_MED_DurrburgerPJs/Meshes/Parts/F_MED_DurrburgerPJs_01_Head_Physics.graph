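# Physics asset for the head part of the female Durrburger PJs cosmetic character, giving secondary motion to the bangs and hoodie eyes.
# A sphere on the head bone serves as the anchor, with spheres on the left and right hoodie eye bones and a capsule on the bangs bone.
# Each dynamic bone is constrained directly to the head with limited swing and a zero-degree twist limit.
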PhysicsAsset /BRCosmetics/Characters/Player/Female/Medium/Bodies/F_MED_DurrburgerPJs/Meshes/Parts/F_MED_DurrburgerPJs_01_Head_Physics
bodies {
 dyn_bangs_01 [{shapeType=Capsule, length=4.797, radius=1.6606, center=<2.5552 0 0.0001>, rotation=<rot -90 0.0049 -0.0049>}]
 dyn_hoodie_eye_l [{shapeType=Sphere, radius=3.4875, center=<0.1 0 0>}]
 dyn_hoodie_eye_r [{shapeType=Sphere, radius=3.4875, center=<0.1 0 0>}]
 head [{shapeType=Sphere, radius=5}]
}
constraints [
 {bone1Name=dyn_hoodie_eye_l, bone2Name=head, swing1LimitDegrees=45, swing2LimitDegrees=45, twistLimitDegrees=0}
 {bone1Name=dyn_hoodie_eye_r, bone2Name=head, swing1LimitDegrees=45, swing2LimitDegrees=45, twistLimitDegrees=0}
 {bone1Name=dyn_bangs_01, bone2Name=head, swing1LimitDegrees=40, swing2LimitDegrees=40, twistLimitDegrees=0}
]
